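# Niagara transition effect for the DoughDisk backpack cosmetic, forming a short burst of tendrils and drops.
# Tendril_Ball bursts black meshes that drift along curves while swelling and shrinking.
# Tendril_Arms bursts random-indexed black meshes, using partition arrays for mirrored orientation and curves for scale and camera offset.
# Drops_ spawns shrinking sprites at a tapering rate, moved by curl noise, attraction and drag, with lifetime driven by owner speed.
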NiagaraSystem /BRCosmetics/Accessories/FORT_Backpacks/Backpack_F_MED_DoughDisk/FX/NS_Backback_DoughDisk_Transition
dependencies {
 usedDataInterfaces [&NiagaraDataInterfaceCurve, &NiagaraDataInterfaceArrayFloat]
 usedDynamicInputs [&SimulationPosition]
 usedModules [&SystemState]
}
emitters {
 Drops_ {
  emitterUpdateScript [
   &EmitterState()
   &SpawnRate(
    SpawnRate=&FloatFromCurve(
     FloatCurve=DataInterface() {propertyValues {Curve {keys [<at 0 1 RCIM_Cubic>, {arriveTangent=-2.3343, interpMode=RCIM_Cubic, leaveTangent=-2.3343, tangentMode=RCTM_User, time=0.7416, value=0}]}}}
     CurveIndex=$Emitter.NormalizedLoopAge
     Scale Curve=25
    )
    Use Spawn Probability=false
   )
  ]
  particleSpawnScript [
   &FN_SetSpawnGroup_v2(Number of Spawn Groups=3) {enabled=false}
   &InitializeParticle(
    Lifetime=&FloatFromCurve(
     FloatCurve=DataInterface() {propertyValues {Curve {keys [<at 0 1 RCIM_Cubic>, {arriveTangent=-1, interpMode=RCIM_Cubic, leaveTangent=-1, time=0.3853, value=0.9371}, <at 1 0 RCIM_Cubic>]}}}
     CurveIndex=&NormalizeFloat(Value=&VectorLength(Vector=$Engine.Owner.Velocity), Min=0, Max=200)
     Scale Curve=2
    )
    Mass Mode=::NewEnumerator1
    Sprite Size=&Vector2DFromFloat(Value=&RandomRangeFloat(Minimum=5, Maximum=3))
    Sprite Size Mode=::NewEnumerator4
    UsePositionOffset=false
    Write Material Random=true
   ) {module=InitializeParticle001}
   /Niagara/Modules/Update/SubUV/SubUVAnimation(SubUV Animation Mode=::NewEnumerator1) {module=SubUVAnimation001}
   &AddVelocityFromPoint(
    Module.Use Falloff Distance=false
    Origin Offset=<0 5 0>
    Velocity Strength=&FloatFromCurve(
     FloatCurve=DataInterface() {propertyValues {Curve {keys [<at 0 1>,<at 0.6 0.2>]}}}
     CurveIndex=$Emitter.NormalizedLoopAge
     Scale Curve=&RandomRangeFloat(Minimum=55, Maximum=100)
    )
   ) {module=AddVelocityFromPoint001}
   &SystemLocation(MaskSpawnGroup=false, Offset=<0 -6 -5>)
  ]
  particleUpdateScript [
   &ParticleState()
   &PointAttractionForce(
    Attraction Radius=350
    Attraction Strength=10
    Falloff Exponent=0.7
    Kill Within Radius=false
    Module.Use Falloff=true
    UseAttractorOffset=false
   )
   &Drag(Drag=2, Use Linear Drag=true, Use Rotational Drag=false)
   &InheritVelocity(
    Module.Limit Speed=true
    Velocity Scale=&VectorFromFloat(Value=&FloatFromCurve(
      FloatCurve=DataInterface() {propertyValues {Curve {keys [{arriveTangent=-1.9956, interpMode=RCIM_Cubic, leaveTangent=-1.9956, tangentMode=RCTM_User, time=0, value=1}, <at 1 0 RCIM_Cubic>]}}}
      Scale Curve=0.1
     ))
   ) {enabled=false, module=InheritVelocity001}
   /Niagara/Modules/Update/Forces/CurlNoiseForce(
    Noise Frequency=0.35
    Noise Strength=&FloatFromCurve(
     FloatCurve=DataInterface() {propertyValues {Curve {keys [{arriveTangent=-2.9606, interpMode=RCIM_Cubic, leaveTangent=-2.9606, tangentMode=RCTM_User, time=0, value=1}, <at 0.5 0 RCIM_Cubic>]}}}
     Scale Curve=250
    )
    Pan Noise=true
    Pan Noise Field=<0 0 2>
   )
   &InheritSourceMovement(Applied Position Scale=<0.5 0.5 0.5>, ApplyForce=false, ApplyPosition=true, ApplyPositionFalloff=false) {enabled=false}
   &InheritSourceMovement(
    Applied Position Scale=&Multiply_VectorByFloat(
     Vector=<0.1 0.1 0.1>
     Float=&FloatFromCurve(FloatCurve=DataInterface() {propertyValues {Curve {keys [
          {interpMode=RCIM_Cubic, tangentMode=RCTM_User, time=0, value=0}
          {arriveTangent=1.7885, interpMode=RCIM_Cubic, leaveTangent=1.7885, tangentMode=RCTM_User, time=1, value=1}
         ]}}})
    )
    ApplyForce=false
    ApplyPosition=true
    ApplyPositionFalloff=false
   ) {module=InheritSourceMovement001}
   &SolveForcesAndVelocity(Clamp Velocity=false, Limit Acceleration=false)
   &ScaleSpriteSize(Scale Factor=&Vector2DFromFloat(Value=&FloatFromCurve(FloatCurve=DataInterface() {propertyValues {Curve {keys [<at 0 1 RCIM_Cubic>, {arriveTangent=-1.7767, interpMode=RCIM_Cubic, leaveTangent=-1.7767, tangentMode=RCTM_User, time=1, value=0}]}}})))
  ]
  rendererClasses [&NiagaraSpriteRendererProperties]
  renderers [{rendererClass=&NiagaraSpriteRendererProperties}]
 }
 Tendril_Arms {
  emitterUpdateScript [
   &EmitterState()
   &SpawnBurst_Instantaneous(Spawn Count=18, Use Spawn Probability=false, UseLoopCountLimit=false)
  ]
  particleSpawnScript [
   /SetVariables(Particles.MeshIndex=&RandomRangeInt(Minimum=0, Maximum=2)) {isSetParameters=true}
   &PartitionParticles(Number Of Partitions=6)
   &InitializeParticle(
    Color=<rgba 0 0 0 1>
    Lifetime Min=1.5
    Lifetime Mode=::NewEnumerator1
    Mesh Scale=&Multiply_VectorByFloat(Vector=<1.1 1.6 1.6>, Float=&RandomRangeFloat(Minimum=0.65, Maximum=0.65))
    Mesh Scale Mode=::NewEnumerator4
    UsePositionOffset=false
    Write Material Random=true
   )
   &InitialMeshOrientation(
    Mesh Axis To Orient=&MakeVector(
     X=&SelectFloatFromArray(
      Direct Array Index=$Particles.PartitionParticles.Partition
      Float Selection Array=DataInterface() {propertyValues {FloatData [-1,1,1,-1,1,-1]}}
     )
     Y=0
     Z=0
    )
    Mesh Orientation Mode=::NewEnumerator2
    Rotation=&MakeVector(
     X=0
     Y=&SelectFloatFromArray(
      Direct Array Index=$Particles.PartitionParticles.Partition
      Float Selection Array=DataInterface() {propertyValues {FloatData [0.13,0.13,0.22,0.22,0.35,0.35]}}
     )
     Z=0
    )
    Use Rotation Vector=true
   )
   &SystemLocation(MaskSpawnGroup=false, Offset=<0 -5 -6>)
  ]
  particleUpdateScript [
   &ParticleState()
   &ScaleMeshSize(Scale Factor=&Multiply_Vector(
     A=&VectorFromFloat(Value=&FloatFromCurve(FloatCurve=DataInterface() {propertyValues {Curve {keys [
           {interpMode=RCIM_Cubic, tangentMode=RCTM_Break, time=0, value=0}
           {arriveTangent=16.9205, interpMode=RCIM_Cubic, leaveTangent=1.2631, tangentMode=RCTM_Break, time=0.15, value=1}
           {arriveTangent=-1.1765, interpMode=RCIM_Cubic, leaveTangent=-1.1765, time=0.7824, value=0.9882}
           {interpMode=RCIM_Cubic, leaveTangent=-3.8739, time=1, value=0}
          ]}}}))
     B=&VectorFromFloat(Value=&FloatFromCurve(
       FloatCurve=DataInterface() {propertyValues {Curve {keys [
           {interpMode=RCIM_Cubic, tangentMode=RCTM_User, time=0, value=0.6}
           {interpMode=RCIM_Cubic, tangentMode=RCTM_User, time=0.5, value=1}
           {tangentMode=RCTM_User, time=1, value=0.5}
          ]}}}
       CurveIndex=&ReturnNormalizedExecIndex()
      ))
    ))
   &CameraOffset(Camera Offset Amount=&FloatFromCurve(
     FloatCurve=DataInterface() {propertyValues {Curve {keys [
         {tangentMode=RCTM_User, time=0, value=1}
         {arriveTangent=-0.0405, interpMode=RCIM_Cubic, leaveTangent=-2.7068, tangentMode=RCTM_Break, time=0.3, value=1}
         {arriveTangent=-0.0538, interpMode=RCIM_Cubic, leaveTangent=-8.0089, tangentMode=RCTM_Break, time=1, value=0}
        ]}}}
     Scale Curve=9
    ))
  ]
  rendererClasses [&NiagaraMeshRendererProperties]
  renderers [
   {rendererClass=&NiagaraMeshRendererProperties, rendererIndex=0}
   {rendererClass=&NiagaraMeshRendererProperties, rendererIndex=1}
  ]
 }
 Tendril_Ball {
  emitterUpdateScript [
   &EmitterState()
   &SpawnBurst_Instantaneous(Spawn Count=5, Use Spawn Probability=false, UseLoopCountLimit=false) {module=SpawnBurst_Instantaneous001}
  ]
  particleSpawnScript [
   &PartitionParticles(Number Of Partitions=5, Partition Mode:ENiagara_PartitionMode=::NewEnumerator5)
   &InitializeParticle(
    Color=<rgba 0 0 0 1>
    Lifetime Max=2.2
    Lifetime Min=1.7
    Lifetime Mode=::NewEnumerator1
    Mesh Scale=&Multiply_VectorByFloat(
     Vector=&RandomRangeVector(Minimum=<1.6 0.8 0.7>, Maximum=<4 1 0.8>)
     Float=&RandomRangeFloat(Minimum=1, Maximum=1)
    )
    Mesh Scale Mode=::NewEnumerator4
    UsePositionOffset=false
    Write Material Random=true
   ) {module=InitializeParticle001}
   &ShapeLocation(MaskSpawnGroup=false, Non Uniform Scale=<2 0 1>, Offset Mode=::NewEnumerator0, Override Seed=false, Sphere Radius=3, UseOwnerScale=true)
   &InitialMeshOrientation(Use Rotation Vector=false)
  ]
  particleUpdateScript [
   &ParticleState()
   /SetVariables(Particles.Position=&AddVectorToPosition(
     Position=Linked() {linkedVariable {name=Particles.ShapeLocation.ShapePosition, type=Vector3f}}
     Vector=&MakeVector(
      X=0
      Y=&FloatFromCurve(FloatCurve=DataInterface() {propertyValues {Curve {keys [
           {interpMode=RCIM_Cubic, tangentMode=RCTM_Break, time=0, value=-7}
           {arriveTangent=6.7186, interpMode=RCIM_Cubic, tangentMode=RCTM_Break, time=1, value=-6}
          ]}}})
      Z=&FloatFromCurve(FloatCurve=DataInterface() {propertyValues {Curve {keys [
           {arriveTangent=-12.825, interpMode=RCIM_Cubic, leaveTangent=-12.825, tangentMode=RCTM_User, time=0, value=0}
           {arriveTangent=0.454, interpMode=RCIM_Cubic, leaveTangent=0.454, tangentMode=RCTM_User, time=1, value=-4}
          ]}}})
     )
    )) {isSetParameters=true}
   &UpdateMeshOrientation()
   &ScaleMeshSize(Scale Factor=&VectorFromFloat(Value=&FloatFromCurve(FloatCurve=DataInterface() {propertyValues {Curve {keys [
          {arriveTangent=6.8033, time=0, value=0}
          {arriveTangent=1.2771, time=0.15, value=1.1}
          <at 0.3 0.8>
          {arriveTangent=0.0366, interpMode=RCIM_Cubic, leaveTangent=-1.0318, tangentMode=RCTM_Break, time=0.33, value=0.5}
          {arriveTangent=0.0984, interpMode=RCIM_Cubic, leaveTangent=-10.6556, tangentMode=RCTM_Break, time=1.0135, value=0.1957}
         ]}}})))
   &CameraOffset(Camera Offset Amount=11)
  ]
  rendererClasses [&NiagaraMeshRendererProperties]
  renderers [{rendererClass=&NiagaraMeshRendererProperties, rendererIndex=0}]
 }
}
summary {
 systemName=NS_Backback_DoughDisk_Transition
 emitters [
  {emitterName=Tendril_Ball, rendererClasses [&NiagaraMeshRendererProperties]}
  {emitterName=Tendril_Arms, rendererClasses [&NiagaraMeshRendererProperties]}
  {emitterName=Drops_, rendererClasses [&NiagaraSpriteRendererProperties]}
 ]
 userVariables [
  User.Color_Tint {type=LinearColor, description="", defaultValue {struct=LinearColor, value=<rgba 0.1128 0.1128 0.1128 1>}}
  User.EmissiveIntensity {type=NiagaraFloat, defaultValue=1, description=""}
  User.SkeletalMesh {
   type=NiagaraDataInterfaceSkeletalMesh
   defaultValue {
    struct=NiagaraExt_VariableValue_DataInterface
    value {dataInterface=&self:NiagaraDataInterfaceSkeletalMesh_0}
   }
  }
 ]
}
