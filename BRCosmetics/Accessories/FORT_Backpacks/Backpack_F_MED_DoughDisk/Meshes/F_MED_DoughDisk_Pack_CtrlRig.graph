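# Control Rig for the DoughDisk backpack, animating its tentacles and two pickaxe bone branches.
# Chain harmonics waves add procedural wiggle to the small tentacles and pickaxe tentacles.
# Transform offsets then shift the pack and pickaxes, weighted by a Kalman-smoothed isSkyGive flag and by remapped spine bend angles from input transforms.
# Finally, parent constraints with graduated weights stretch the larger tentacle chains between Big_pice and the pickaxe on each side.
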ControlRigBlueprint /BRCosmetics/Accessories/FORT_Backpacks/Backpack_F_MED_DoughDisk/Meshes/F_MED_DoughDisk_Pack_CtrlRig
graphs [
 RigVMFunctionLibrary {links [], nodes []}
 RigVMFunction_Sequence_1_ContainedGraph {nodes [
   n1 = RigVMFunctionEntryNode()
   RigVMFunction_Sequence = &RigVMFunction_Sequence(ExecuteContext=@n1.ExecuteContext)
   n2 = &RigVMFunction_Sequence(ExecuteContext=@RigVMFunction_Sequence.B)
   RigVMFunctionReturnNode(A=@RigVMFunction_Sequence.A, B=@n2.A, C=@n2.B)
  ]}
 RigVMModel {nodes [
   RigUnit_BeginExecution = &RigUnit_BeginExecution()
   RigVMFunction_Sequence = RigVMAggregateNode(ExecuteContext=@RigUnit_BeginExecution.ExecutePin)
   /Script/ControlRig.RigUnit_ChainHarmonicsPerItem(
    ChainRoot=(Type=Bone,Name="tentacle_upper_01_l_smol")
    DrawWorldOffset=(Rotation=(X=0,Y=0,Z=0,W=1),Translation=(X=0,Y=0,Z=0),Scale3D=(X=1,Y=1,Z=1))
    Pendulum=(bEnabled=False,PendulumStiffness=2,PendulumGravity=(X=0,Y=0,Z=0),PendulumBlend=0.75,PendulumDrag=0.98,PendulumMinimum=0,PendulumMaximum=0.1,PendulumEase=Linear,UnwindAxis=(X=0,Y=1,Z=0),UnwindMinimum=0.2,UnwindMaximum=0.05)
    Reach=(bEnabled=False,ReachTarget=(X=0,Y=0,Z=0),ReachAxis=(X=1,Y=0,Z=0),ReachMinimum=0,ReachMaximum=0,ReachEase=Linear)
    Speed=(X=1,Y=1,Z=0.7)
    Wave=(bEnabled=True,WaveFrequency=(X=1,Y=0.8,Z=-3),WaveAmplitude=(X=0,Y=0,Z=1.5),WaveOffset=(X=0,Y=1,Z=1),WaveNoise=(X=0,Y=0,Z=0),WaveMinimum=0,WaveMaximum=1,WaveEase=Linear)
    WaveCurve=(EditorCurveData=(Keys=((),(Time=0.3914,Value=0.7293))))
    bDrawDebug=false
    ExecutePin=@RigVMFunction_Sequence.A
   ) {is=struct RigVMUnitNode}
   > &RigUnit_ChainHarmonicsPerItem(
    ChainRoot=(Type=Bone,Name="tentacle_mid_01_l_smol")
    DrawWorldOffset=(Rotation=(X=0,Y=0,Z=0,W=1),Translation=(X=0,Y=0,Z=0),Scale3D=(X=1,Y=1,Z=1))
    Pendulum=(bEnabled=False,PendulumStiffness=2,PendulumGravity=(X=0,Y=0,Z=0),PendulumBlend=0.75,PendulumDrag=0.98,PendulumMinimum=0,PendulumMaximum=0.1,PendulumEase=Linear,UnwindAxis=(X=0,Y=1,Z=0),UnwindMinimum=0.2,UnwindMaximum=0.05)
    Reach=(bEnabled=False,ReachTarget=(X=0,Y=0,Z=0),ReachAxis=(X=1,Y=0,Z=0),ReachMinimum=0,ReachMaximum=0,ReachEase=Linear)
    Speed=(X=1.1,Y=1.1,Z=1.1)
    Wave=(bEnabled=True,WaveFrequency=(X=1,Y=0.8,Z=-3),WaveAmplitude=(X=0,Y=0,Z=0.8),WaveOffset=(X=0,Y=1,Z=1),WaveNoise=(X=0,Y=0,Z=0),WaveMinimum=0,WaveMaximum=1,WaveEase=Linear)
    WaveCurve=(EditorCurveData=(Keys=((),(InterpMode=RCIM_Constant,Time=0.6195,Value=0.9894))))
    bDrawDebug=false
   )
   > &RigUnit_ChainHarmonicsPerItem(
    ChainRoot=(Type=Bone,Name="tentacle_bot_01_l_smol")
    DrawWorldOffset=(Rotation=(X=0,Y=0,Z=0,W=1),Translation=(X=0,Y=0,Z=0),Scale3D=(X=1,Y=1,Z=1))
    Pendulum=(bEnabled=False,PendulumStiffness=2,PendulumGravity=(X=0,Y=0,Z=0),PendulumBlend=0.75,PendulumDrag=0.98,PendulumMinimum=0,PendulumMaximum=0.1,PendulumEase=Linear,UnwindAxis=(X=0,Y=1,Z=0),UnwindMinimum=0.2,UnwindMaximum=0.05)
    Reach=(bEnabled=False,ReachTarget=(X=0,Y=0,Z=0),ReachAxis=(X=1,Y=0,Z=0),ReachMinimum=0,ReachMaximum=0,ReachEase=Linear)
    Speed=(X=1,Y=1,Z=1.1)
    Wave=(bEnabled=true,WaveFrequency=(X=-1,Y=0.8,Z=-3),WaveAmplitude=(X=0,Y=0,Z=2),WaveOffset=(X=0,Y=1,Z=-0.5),WaveNoise=(X=0,Y=0,Z=0.25),WaveMinimum=0,WaveMaximum=1,WaveEase=Linear)
    WaveCurve=(EditorCurveData=(Keys=((),(Time=0.498,Value=0.452))))
    bDrawDebug=false
   )
   > &RigUnit_ChainHarmonicsPerItem(
    ChainRoot=(Type=Bone,Name="tentacle_bot_01_r_smol")
    DrawWorldOffset=(Rotation=(X=0,Y=0,Z=0,W=1),Translation=(X=0,Y=0,Z=0),Scale3D=(X=1,Y=1,Z=1))
    Pendulum=(bEnabled=False,PendulumStiffness=2,PendulumGravity=(X=0,Y=0,Z=0),PendulumBlend=0.75,PendulumDrag=0.98,PendulumMinimum=0,PendulumMaximum=0.1,PendulumEase=Linear,UnwindAxis=(X=0,Y=1,Z=0),UnwindMinimum=0.2,UnwindMaximum=0.05)
    Reach=(bEnabled=False,ReachTarget=(X=0,Y=0,Z=0),ReachAxis=(X=1,Y=0,Z=0),ReachMinimum=0,ReachMaximum=0,ReachEase=Linear)
    Speed=(X=1,Y=1,Z=1.5)
    Wave=(bEnabled=True,WaveFrequency=(X=1,Y=0.8,Z=-3),WaveAmplitude=(X=0,Y=0,Z=1),WaveOffset=(X=0,Y=1,Z=0.5),WaveNoise=(X=0,Y=0,Z=-0.1),WaveMinimum=0,WaveMaximum=1,WaveEase=Linear)
    WaveCurve=(EditorCurveData=(Keys=((),(Time=0.3748,Value=0.5122))))
    bDrawDebug=false
   )
   > &RigUnit_ChainHarmonicsPerItem(
    ChainRoot=(Type=Bone,Name="tentacle_midup_01_r_smol")
    DrawWorldOffset=(Rotation=(X=0,Y=0,Z=0,W=1),Translation=(X=0,Y=0,Z=0),Scale3D=(X=1,Y=1,Z=1))
    Pendulum=(bEnabled=False,PendulumStiffness=2,PendulumGravity=(X=0,Y=0,Z=0),PendulumBlend=0.75,PendulumDrag=0.98,PendulumMinimum=0,PendulumMaximum=0.1,PendulumEase=Linear,UnwindAxis=(X=0,Y=1,Z=0),UnwindMinimum=0.2,UnwindMaximum=0.05)
    Reach=(bEnabled=False,ReachTarget=(X=0,Y=0,Z=0),ReachAxis=(X=1,Y=0,Z=0),ReachMinimum=0,ReachMaximum=0,ReachEase=Linear)
    Speed=(X=1,Y=1,Z=1.5)
    Wave=(bEnabled=True,WaveFrequency=(X=1,Y=0.8,Z=-3),WaveAmplitude=(X=0,Y=0,Z=1),WaveOffset=(X=0,Y=1,Z=0.5),WaveNoise=(X=0,Y=0,Z=-0.1),WaveMinimum=0,WaveMaximum=1,WaveEase=Linear)
    WaveCurve=(EditorCurveData=(Keys=((),(Time=0.3748,Value=0.5122))))
    bDrawDebug=false
   )
   > &RigUnit_ChainHarmonicsPerItem(
    ChainRoot=(Type=Bone,Name="tentacle_upper_01_r_smol")
    DrawWorldOffset=(Rotation=(X=0,Y=0,Z=0,W=1),Translation=(X=0,Y=0,Z=0),Scale3D=(X=1,Y=1,Z=1))
    Pendulum=(bEnabled=False,PendulumStiffness=2,PendulumGravity=(X=0,Y=0,Z=0),PendulumBlend=0.75,PendulumDrag=0.98,PendulumMinimum=0,PendulumMaximum=0.1,PendulumEase=Linear,UnwindAxis=(X=0,Y=1,Z=0),UnwindMinimum=0.2,UnwindMaximum=0.05)
    Reach=(bEnabled=False,ReachTarget=(X=0,Y=0,Z=0),ReachAxis=(X=1,Y=0,Z=0),ReachMinimum=0,ReachMaximum=0,ReachEase=Linear)
    Speed=(X=1,Y=1,Z=1)
    Wave=(bEnabled=True,WaveFrequency=(X=1,Y=0.8,Z=-3),WaveAmplitude=(X=0,Y=0,Z=1.5),WaveOffset=(X=0,Y=1,Z=1),WaveNoise=(X=0,Y=0,Z=0),WaveMinimum=0,WaveMaximum=0.7,WaveEase=Linear)
    WaveCurve=(EditorCurveData=(Keys=((),(InterpMode=RCIM_Cubic,Time=0.3739,Value=0.5254))))
    bDrawDebug=false
   )
   > &RigUnit_ChainHarmonicsPerItem(
    ChainRoot=(Type=Bone,Name="tentacle_harm_01")
    DrawWorldOffset=(Rotation=(X=0,Y=0,Z=0,W=1),Translation=(X=0,Y=0,Z=0),Scale3D=(X=1,Y=1,Z=1))
    Pendulum=(bEnabled=False,PendulumStiffness=2,PendulumGravity=(X=0,Y=0,Z=0),PendulumBlend=0.75,PendulumDrag=0.98,PendulumMinimum=0,PendulumMaximum=0.1,PendulumEase=Linear,UnwindAxis=(X=0,Y=1,Z=0),UnwindMinimum=0.2,UnwindMaximum=0.05)
    Reach=(bEnabled=False,ReachTarget=(X=0,Y=0,Z=0),ReachAxis=(X=1,Y=0,Z=0),ReachMinimum=0,ReachMaximum=0,ReachEase=Linear)
    Speed=(X=1,Y=1,Z=1)
    Wave=(bEnabled=True,WaveFrequency=(X=1,Y=0.8,Z=-3),WaveAmplitude=(X=1,Y=1,Z=1),WaveOffset=(X=0,Y=1,Z=1),WaveNoise=(X=0,Y=0,Z=0.3),WaveMinimum=0,WaveMaximum=1.1,WaveEase=Linear)
    WaveCurve=(EditorCurveData=(Keys=((),(InterpMode=RCIM_Constant,Time=0.5874,Value=0.2501))))
    bDrawDebug=false
   )
   > &RigUnit_ChainHarmonicsPerItem(
    ChainRoot=(Type=Bone,Name="tentacle_middv_01_r_smol")
    DrawWorldOffset=(Rotation=(X=0,Y=0,Z=0,W=1),Translation=(X=0,Y=0,Z=0),Scale3D=(X=1,Y=1,Z=1))
    Pendulum=(bEnabled=False,PendulumStiffness=2,PendulumGravity=(X=0,Y=0,Z=0),PendulumBlend=0.75,PendulumDrag=0.98,PendulumMinimum=0,PendulumMaximum=0.1,PendulumEase=Linear,UnwindAxis=(X=0,Y=1,Z=0),UnwindMinimum=0.2,UnwindMaximum=0.05)
    Reach=(bEnabled=False,ReachTarget=(X=0,Y=0,Z=0),ReachAxis=(X=1,Y=0,Z=0),ReachMinimum=0,ReachMaximum=0,ReachEase=Linear)
    Speed=(X=0.8,Y=1,Z=1)
    Wave=(bEnabled=True,WaveFrequency=(X=1,Y=0.8,Z=-3),WaveAmplitude=(X=0,Y=0,Z=1.5),WaveOffset=(X=0,Y=1,Z=1),WaveNoise=(X=0,Y=0,Z=0),WaveMinimum=0,WaveMaximum=1,WaveEase=Linear)
    WaveCurve=(EditorCurveData=(Keys=((),(InterpMode=RCIM_Constant,Time=0.5756,Value=0.4391))))
    bDrawDebug=false
   )
   &RigUnit_ChainHarmonicsPerItem(
    ChainRoot=(Type=Bone,Name="Ppicaxe_tentkl_01_r_base")
    DrawWorldOffset=(Rotation=(X=0,Y=0,Z=0,W=1),Translation=(X=0,Y=0,Z=0),Scale3D=(X=1,Y=1,Z=1))
    Pendulum=(bEnabled=False,PendulumStiffness=2,PendulumGravity=(X=0,Y=0,Z=0),PendulumBlend=0.75,PendulumDrag=0.98,PendulumMinimum=0,PendulumMaximum=0.1,PendulumEase=Linear,UnwindAxis=(X=0,Y=1,Z=0),UnwindMinimum=0.2,UnwindMaximum=0.05)
    Reach=(bEnabled=False,ReachTarget=(X=0,Y=0,Z=0),ReachAxis=(X=1,Y=0,Z=0),ReachMinimum=0,ReachMaximum=0,ReachEase=Linear)
    Speed=(X=1,Y=0.8,Z=1)
    Wave=(bEnabled=True,WaveFrequency=(X=1,Y=1,Z=-1),WaveAmplitude=(X=0,Y=1,Z=0.3),WaveOffset=(X=0,Y=1,Z=0.5),WaveNoise=(X=0,Y=0.3,Z=0),WaveMinimum=0,WaveMaximum=1,WaveEase=Linear)
    WaveCurve=(EditorCurveData=(Keys=((),(Time=0.3691,Value=0.3527))))
    bDrawDebug=false
    ExecutePin=@RigVMFunction_Sequence.B
   )
   > &RigUnit_ChainHarmonicsPerItem(
    ChainRoot=(Type=Bone,Name="Ppicaxe_tentkl_03_r_base")
    DrawWorldOffset=(Rotation=(X=0,Y=0,Z=0,W=1),Translation=(X=0,Y=0,Z=0),Scale3D=(X=1,Y=1,Z=1))
    Pendulum=(bEnabled=False,PendulumStiffness=2,PendulumGravity=(X=0,Y=0,Z=0),PendulumBlend=0.75,PendulumDrag=0.98,PendulumMinimum=0,PendulumMaximum=0.1,PendulumEase=Linear,UnwindAxis=(X=0,Y=1,Z=0),UnwindMinimum=0.2,UnwindMaximum=0.05)
    Reach=(bEnabled=False,ReachTarget=(X=0,Y=0,Z=0),ReachAxis=(X=1,Y=0,Z=0),ReachMinimum=0,ReachMaximum=0,ReachEase=Linear)
    Speed=(X=0.3,Y=0.7,Z=1)
    Wave=(bEnabled=True,WaveFrequency=(X=1,Y=0.8,Z=-3),WaveAmplitude=(X=0,Y=0.8,Z=0.2),WaveOffset=(X=0,Y=1,Z=1),WaveNoise=(X=0,Y=0,Z=0),WaveMinimum=0,WaveMaximum=1,WaveEase=Linear)
    WaveCurve=(EditorCurveData=(Keys=((),(Time=0.3843,Value=0.3539))))
    bDrawDebug=false
   )
   > &RigUnit_ChainHarmonicsPerItem(
    ChainRoot=(Type=Bone,Name="Ppicaxe_tentkl_05_r_base")
    DrawWorldOffset=(Rotation=(X=0,Y=0,Z=0,W=1),Translation=(X=0,Y=0,Z=0),Scale3D=(X=1,Y=1,Z=1))
    Pendulum=(bEnabled=False,PendulumStiffness=2,PendulumGravity=(X=0,Y=0,Z=0),PendulumBlend=0.75,PendulumDrag=0.98,PendulumMinimum=0,PendulumMaximum=0.1,PendulumEase=Linear,UnwindAxis=(X=0,Y=1,Z=0),UnwindMinimum=0.2,UnwindMaximum=0.05)
    Reach=(bEnabled=False,ReachTarget=(X=0,Y=0,Z=0),ReachAxis=(X=1,Y=0,Z=0),ReachMinimum=0,ReachMaximum=0,ReachEase=Linear)
    Speed=(X=1,Y=0.7,Z=1)
    Wave=(bEnabled=True,WaveFrequency=(X=1,Y=0.8,Z=-3),WaveAmplitude=(X=0,Y=0.7,Z=0.6),WaveOffset=(X=0,Y=1,Z=1),WaveNoise=(X=0,Y=0.3,Z=0.1),WaveMinimum=0,WaveMaximum=1,WaveEase=Linear)
    WaveCurve=(EditorCurveData=(Keys=((),(InterpMode=RCIM_Constant,Time=0.599,Value=0.5202))))
    bDrawDebug=false
   )
   > &RigUnit_ChainHarmonicsPerItem(
    ChainRoot=(Type=Bone,Name="Ppicaxe_tentkl_06_l_base")
    DrawWorldOffset=(Rotation=(X=0,Y=0,Z=0,W=1),Translation=(X=0,Y=0,Z=0),Scale3D=(X=1,Y=1,Z=1))
    Pendulum=(bEnabled=False,PendulumStiffness=2,PendulumGravity=(X=0,Y=0,Z=0),PendulumBlend=0.75,PendulumDrag=0.98,PendulumMinimum=0,PendulumMaximum=0.1,PendulumEase=Linear,UnwindAxis=(X=0,Y=1,Z=0),UnwindMinimum=0.2,UnwindMaximum=0.05)
    Reach=(bEnabled=False,ReachTarget=(X=0,Y=0,Z=0),ReachAxis=(X=1,Y=0,Z=0),ReachMinimum=0,ReachMaximum=0,ReachEase=Linear)
    Speed=(X=1,Y=1,Z=1)
    Wave=(bEnabled=True,WaveFrequency=(X=1,Y=0.8,Z=-3),WaveAmplitude=(X=0,Y=0.8,Z=0.5),WaveOffset=(X=0,Y=1,Z=0.5),WaveNoise=(X=0,Y=0,Z=-0.1),WaveMinimum=0,WaveMaximum=1,WaveEase=Linear)
    WaveCurve=(EditorCurveData=(Keys=((),(Time=0.3588,Value=0.288))))
    bDrawDebug=false
   )
   > &RigUnit_ChainHarmonicsPerItem(
    ChainRoot=(Type=Bone,Name="Ppicaxe_tentkl_04_l_base")
    DrawWorldOffset=(Rotation=(X=0,Y=0,Z=0,W=1),Translation=(X=0,Y=0,Z=0),Scale3D=(X=1,Y=1,Z=1))
    Pendulum=(bEnabled=False,PendulumStiffness=2,PendulumGravity=(X=0,Y=0,Z=0),PendulumBlend=0.75,PendulumDrag=0.98,PendulumMinimum=0,PendulumMaximum=0.1,PendulumEase=Linear,UnwindAxis=(X=0,Y=1,Z=0),UnwindMinimum=0.2,UnwindMaximum=0.05)
    Reach=(bEnabled=False,ReachTarget=(X=0,Y=0,Z=0),ReachAxis=(X=1,Y=0,Z=0),ReachMinimum=0,ReachMaximum=0,ReachEase=Linear)
    Speed=(X=1,Y=1.1,Z=0.6)
    Wave=(bEnabled=true,WaveFrequency=(X=1,Y=0.8,Z=-3),WaveAmplitude=(X=0,Y=0.8,Z=0.5),WaveOffset=(X=0,Y=1,Z=1),WaveNoise=(X=0,Y=0,Z=0),WaveMinimum=0,WaveMaximum=1,WaveEase=Linear)
    WaveCurve=(EditorCurveData=(Keys=((),(Time=0.1765,Value=0.5113))))
    bDrawDebug=false
   )
   > &RigUnit_ChainHarmonicsPerItem(
    ChainRoot=(Type=Bone,Name="Ppicaxe_tentkl_02_l_base")
    DrawWorldOffset=(Rotation=(X=0,Y=0,Z=0,W=1),Translation=(X=0,Y=0,Z=0),Scale3D=(X=1,Y=1,Z=1))
    Pendulum=(bEnabled=False,PendulumStiffness=2,PendulumGravity=(X=0,Y=0,Z=0),PendulumBlend=0.75,PendulumDrag=0.98,PendulumMinimum=0,PendulumMaximum=0.1,PendulumEase=Linear,UnwindAxis=(X=0,Y=1,Z=0),UnwindMinimum=0.2,UnwindMaximum=0.05)
    Reach=(bEnabled=False,ReachTarget=(X=0,Y=0,Z=0),ReachAxis=(X=1,Y=0,Z=0),ReachMinimum=0,ReachMaximum=0,ReachEase=Linear)
    Speed=(X=0.8,Y=1,Z=0.8)
    Wave=(bEnabled=True,WaveFrequency=(X=1,Y=0.8,Z=-3),WaveAmplitude=(X=0,Y=0.8,Z=0.5),WaveOffset=(X=0,Y=1,Z=1),WaveNoise=(X=0,Y=0,Z=0),WaveMinimum=0,WaveMaximum=1,WaveEase=Linear)
    WaveCurve=(EditorCurveData=(Keys=((),(InterpMode=RCIM_Constant,Time=0.5369,Value=0.4036))))
    bDrawDebug=false
   )
   RigVMFunction_MathBoolToFloat = &RigVMFunction_MathBoolToFloat(Value=RigVMVariableNode() {variable=isSkyGive}.Value)
   KalmanFilter = &RigVMFunction_KalmanFloat(BufferSize=16, Value=@RigVMFunction_MathBoolToFloat.Result) {template="KalmanFilter::Execute(in Value,in BufferSize,out Result)"}
   &RigUnit_ModifyTransforms(
    ItemToModify="((Item=(Type=Bone,Name=\"Ppicaxe_01_l\"),Transform=(Rotation=(X=0,Y=0,Z=0.3007,W=0.9537),Translation=(X=0,Y=-2,Z=0),Scale3D=(X=1,Y=1,Z=1))),(Item=(Type=Bone,Name=\"Ppicaxe_01_r\"),Transform=(Rotation=(X=0,Y=0,Z=-0.3007,W=0.9537),Translation=(X=0,Y=-2,Z=0),Scal ...(truncated)"
    WeightMinimum=0
    Weight=@KalmanFilter.Result
    ExecutePin=@RigVMFunction_Sequence.C
   )
   Cast = /Script/RigVM.RigVMFunction_MathQuaternionToRotator(Value=&RigVMFunction_MathTransformMakeRelative(Parent=RigVMVariableNode() {variable=spine_05}.Value, Global=RigVMVariableNode() {variable=pelvis}.Value).Local.Rotation) {is=struct RigVMUnitNode, template="Cast::Execute(in Value,out Result)"}
   DISPATCH_RigVMDispatch_BreakStruct = RigVMDispatchNode(Struct=@Cast.Result) {template="DISPATCH_RigVMDispatch_BreakStruct(in Struct,out Elements)"}
   Remap = &RigVMFunction_MathDoubleRemap(SourceMaximum=-45, SourceMinimum=-20, TargetMaximum=1, TargetMinimum=0, Value=@DISPATCH_RigVMDispatch_BreakStruct.Elements.Roll)
   > &RigUnit_ModifyTransforms(
    ItemToModify="((Item=(Type=Bone,Name=\"Ppicaxe_01_r\"),Transform=(Rotation=(X=0,Y=0,Z=-0.3007,W=0.9537),Translation=(X=1,Y=0,Z=0),Scale3D=(X=1,Y=1,Z=1))),(Item=(Type=Bone,Name=\"pack\"),Transform=(Rotation=(X=0,Y=0,Z=-0.0872,W=0.9962),Translation=(X=0,Y=0,Z=0),Scale3D=(X=1, ...(truncated)"
    WeightMinimum=0
    Weight=@Remap.Result
   )
   n1 = &RigVMFunction_MathQuaternionToRotator(Value=&RigVMFunction_MathTransformMakeRelative(Global=RigVMVariableNode() {variable=spine_04}.Value, Parent=RigVMVariableNode() {variable=spine_03}.Value).Local.Rotation) {template="Cast::Execute(in Value,out Result)"}
   > &RigUnit_ModifyTransforms(
    ItemToModify=((Item=(Type=Bone,Name="pack"),Transform=(Rotation=(X=-0.2164,Y=0,Z=0,W=0.9763),Translation=(X=0,Y=-2,Z=8),Scale3D=(X=1,Y=1,Z=1))))
    WeightMinimum=0
    Weight=&RigVMFunction_MathDoubleRemap(
     SourceMaximum=-15
     SourceMinimum=5
     TargetMaximum=1
     TargetMinimum=0
     Value=RigVMDispatchNode(Struct=@n1.Result) {template="DISPATCH_RigVMDispatch_BreakStruct(in Struct,out Elements)"}.Elements.Yaw
    ).Result
   )
   > &RigUnit_ParentConstraint(
    AdvancedSettings=(InterpolationType=Average,RotationOrderForFilter=XZY)
    Child=(Type=Bone,Name="tentacle_upper_02_r")
    Parents=((Item=(Type=Bone,Name="Ppicaxe_01_r"),Weight=0.25),(Item=(Type=Bone,Name="Big_pice"),Weight=0.75))
    Weight=1
   )
   > &RigUnit_ParentConstraint(
    AdvancedSettings=(InterpolationType=Average,RotationOrderForFilter=XZY)
    Child=(Type=Bone,Name="tentacle_upper_03_r")
    Parents=((Item=(Type=Bone,Name="Ppicaxe_01_r"),Weight=0.5),(Item=(Type=Bone,Name="Big_pice"),Weight=0.5))
    Weight=1
   )
   > &RigUnit_ParentConstraint(
    AdvancedSettings=(InterpolationType=Average,RotationOrderForFilter=XZY)
    Child=(Type=Bone,Name="tentacle_upper_04_r")
    Parents=((Item=(Type=Bone,Name="Ppicaxe_01_r"),Weight=1))
    Weight=1
   )
   > &RigUnit_ParentConstraint(
    AdvancedSettings=(InterpolationType=Average,RotationOrderForFilter=XZY)
    Child=(Type=Bone,Name="tentacle_midup_02_r")
    Parents=((Item=(Type=Bone,Name="Ppicaxe_01_r"),Weight=0.25),(Item=(Type=Bone,Name="Big_pice"),Weight=0.75))
    Weight=1
   )
   > &RigUnit_ParentConstraint(
    AdvancedSettings=(InterpolationType=Average,RotationOrderForFilter=XZY)
    Child=(Type=Bone,Name="tentacle_midup_03_r")
    Parents=((Item=(Type=Bone,Name="Ppicaxe_01_r"),Weight=0.5),(Item=(Type=Bone,Name="Big_pice"),Weight=0.5))
    Weight=1
   )
   > &RigUnit_ParentConstraint(
    AdvancedSettings=(InterpolationType=Average,RotationOrderForFilter=XZY)
    Child=(Type=Bone,Name="tentacle_midup_04_r")
    Parents=((Item=(Type=Bone,Name="Ppicaxe_01_r"),Weight=0.75))
    Weight=1
   )
   > &RigUnit_ParentConstraint(
    AdvancedSettings=(InterpolationType=Average,RotationOrderForFilter=XZY)
    Child=(Type=Bone,Name="tentacle_middv_02_r")
    Parents=((Item=(Type=Bone,Name="Ppicaxe_01_r"),Weight=0.25),(Item=(Type=Bone,Name="Big_pice"),Weight=0.75))
    Weight=1
   )
   > &RigUnit_ParentConstraint(
    AdvancedSettings=(InterpolationType=Average,RotationOrderForFilter=XZY)
    Child=(Type=Bone,Name="tentacle_middv_03_r")
    Parents=((Item=(Type=Bone,Name="Ppicaxe_01_r"),Weight=0.5),(Item=(Type=Bone,Name="Big_pice"),Weight=0.5))
    Weight=1
   )
   > &RigUnit_ParentConstraint(
    AdvancedSettings=(InterpolationType=Average,RotationOrderForFilter=XZY)
    Child=(Type=Bone,Name="tentacle_middv_04_r")
    Parents=((Item=(Type=Bone,Name="Ppicaxe_01_r"),Weight=0.75))
    Weight=1
   )
   > &RigUnit_ParentConstraint(
    AdvancedSettings=(InterpolationType=Average,RotationOrderForFilter=XZY)
    Child=(Type=Bone,Name="tentacle_bot_02_r")
    Parents=((Item=(Type=Bone,Name="Ppicaxe_01_r"),Weight=0.25),(Item=(Type=Bone,Name="Big_pice"),Weight=0.75))
    Weight=1
   )
   > &RigUnit_ParentConstraint(
    AdvancedSettings=(InterpolationType=Average,RotationOrderForFilter=XZY)
    Child=(Type=Bone,Name="tentacle_bot_03_r")
    Parents=((Item=(Type=Bone,Name="Ppicaxe_01_r"),Weight=0.5),(Item=(Type=Bone,Name="Big_pice"),Weight=0.5))
    Weight=1
   )
   > &RigUnit_ParentConstraint(
    AdvancedSettings=(InterpolationType=Average,RotationOrderForFilter=XZY)
    Child=(Type=Bone,Name="tentacle_bot_04_r")
    Parents=((Item=(Type=Bone,Name="Ppicaxe_01_r"),Weight=0.75))
    Weight=1
   )
   > &RigUnit_ParentConstraint(
    AdvancedSettings=(InterpolationType=Average,RotationOrderForFilter=XZY)
    Child=(Type=Bone,Name="tentacle_upper_02_l")
    Parents=((Item=(Type=Bone,Name="Ppicaxe_01_l"),Weight=0.25),(Item=(Type=Bone,Name="Big_pice"),Weight=0.75))
    Weight=1
   )
   > &RigUnit_ParentConstraint(
    AdvancedSettings=(InterpolationType=Average,RotationOrderForFilter=XZY)
    Child=(Type=Bone,Name="tentacle_upper_03_l")
    Parents=((Item=(Type=Bone,Name="Ppicaxe_01_l"),Weight=0.5),(Item=(Type=Bone,Name="Big_pice"),Weight=0.5))
    Weight=1
   )
   > &RigUnit_ParentConstraint(
    AdvancedSettings=(InterpolationType=Average,RotationOrderForFilter=XZY)
    Child=(Type=Bone,Name="tentacle_upper_04_l")
    Parents=((Item=(Type=Bone,Name="Ppicaxe_01_l"),Weight=0.75))
    Weight=1
   )
   > &RigUnit_ParentConstraint(
    AdvancedSettings=(InterpolationType=Average,RotationOrderForFilter=XZY)
    Child=(Type=Bone,Name="tentacle_mid_02_l")
    Parents=((Item=(Type=Bone,Name="Ppicaxe_01_l"),Weight=0.25),(Item=(Type=Bone,Name="Big_pice"),Weight=0.75))
    Weight=1
   )
   > &RigUnit_ParentConstraint(
    AdvancedSettings=(InterpolationType=Average,RotationOrderForFilter=XZY)
    Child=(Type=Bone,Name="tentacle_mid_03_l")
    Parents=((Item=(Type=Bone,Name="Ppicaxe_01_l"),Weight=0.5),(Item=(Type=Bone,Name="Big_pice"),Weight=0.5))
    Weight=1
   )
   > &RigUnit_ParentConstraint(
    AdvancedSettings=(InterpolationType=Average,RotationOrderForFilter=XZY)
    Child=(Type=Bone,Name="tentacle_mid_04_l")
    Parents=((Item=(Type=Bone,Name="Ppicaxe_01_l"),Weight=0.75),(Item=(Type=Bone,Name="Big_pice"),Weight=0))
    Weight=1
   )
   > &RigUnit_ParentConstraint(
    AdvancedSettings=(InterpolationType=Average,RotationOrderForFilter=XZY)
    Child=(Type=Bone,Name="tentacle_bot_02_l")
    Parents=((Item=(Type=Bone,Name="Ppicaxe_01_l"),Weight=0.25),(Item=(Type=Bone,Name="Big_pice"),Weight=0.75))
    Weight=1
   )
   > &RigUnit_ParentConstraint(
    AdvancedSettings=(InterpolationType=Average,RotationOrderForFilter=XZY)
    Child=(Type=Bone,Name="tentacle_bot_03_l")
    Parents=((Item=(Type=Bone,Name="Ppicaxe_01_l"),Weight=0.5),(Item=(Type=Bone,Name="Big_pice"),Weight=0.5))
    Weight=1
   )
   > &RigUnit_ParentConstraint(
    AdvancedSettings=(InterpolationType=Average,RotationOrderForFilter=XZY)
    Child=(Type=Bone,Name="tentacle_bot_04_l")
    Parents=((Item=(Type=Bone,Name="Ppicaxe_01_l"),Weight=0.75))
    Weight=1
   )
  ]}
]
hierarchy [
 root {loc=<0 0 0>, rot=<0 0 0 1>}
 pack {parent=root, loc=<0 -0.7316 0>, rot=<0 0 0 1>}
 pack1 {parent=pack, loc=<0 -2.1001 0>, rot=<0 0 0 1>}
 Big_pice {parent=pack1, loc=<0 -1.1013 0>, rot=<0 0 0 1>}
 tentacle_harm_01 {parent=Big_pice, loc=<0.338 0.7117 -11.1529>, rot=<0.5 -0.5 -0.5 -0.5>}
 tentacle_harm_02 {parent=tentacle_harm_01, loc=<1.6407 0 0>, rot=<-0.0916 0.2481 -0.2328 -0.9359>}
 tentacle_harm_03 {parent=tentacle_harm_02, loc=<1.4597 0 0>, rot=<0 0 -0.1689 -0.9856>}
 tentacle_middv_01_r {parent=Big_pice, loc=<4.9218 0.3032 -5.9134>, rot=<0.6967 -0.1894 -0.1783 -0.6685>}
 tentacle_middv_02_r {parent=tentacle_middv_01_r, loc=<2.0497 0 0>, rot=<0 0 0 1>}
 tentacle_middv_03_r {parent=tentacle_middv_02_r, loc=<2.2485 0 0>, rot=<0 0 0 1>}
 tentacle_middv_04_r {parent=tentacle_middv_03_r, loc=<1.7627 0 0>, rot=<0 0 0 1>}
 tentacle_middv_05_r {parent=tentacle_middv_04_r, loc=<1.8327 0 0>, rot=<0 0 0 1>}
 tentacle_bot_01_r {parent=Big_pice, loc=<3.7287 0.1797 -8.9464>, rot=<0.6619 -0.2884 -0.2733 -0.6356>}
 tentacle_bot_02_r {parent=tentacle_bot_01_r, loc=<2.3859 0 0>, rot=<0 0 0 1>}
 tentacle_bot_03_r {parent=tentacle_bot_02_r, loc=<1.4815 0 0>, rot=<0 0 0 1>}
 tentacle_bot_04_r {parent=tentacle_bot_03_r, loc=<1.992 0 0>, rot=<0 0 0 1>}
 tentacle_bot_05_r {parent=tentacle_bot_04_r, loc=<1.9734 0 0>, rot=<0 0 0 1>}
 tentacle_bot_01_l {parent=Big_pice, loc=<-4.0705 0.4381 -7.7796>, rot=<-0.2857 0.6525 0.6545 0.2535>}
 tentacle_bot_02_l {parent=tentacle_bot_01_l, loc=<2.5526 0 0>, rot=<0 0 0 1>}
 tentacle_bot_03_l {parent=tentacle_bot_02_l, loc=<1.9604 0 0>, rot=<0 0 0 1>}
 tentacle_bot_04_l {parent=tentacle_bot_03_l, loc=<1.7353 0.0079 0>, rot=<0 0 0 1>}
 tentacle_bot_05_l {parent=tentacle_bot_04_l, loc=<1.7662 0 0>, rot=<0 0 0 1>}
 tentacle_mid_01_l {parent=Big_pice, loc=<-5.6652 0.6408 -3.7709>, rot=<-0.2857 0.6525 0.6545 0.2535>}
 tentacle_mid_02_l {parent=tentacle_mid_01_l, loc=<2.5526 0 0>, rot=<0 0 0 1>}
 tentacle_mid_03_l {parent=tentacle_mid_02_l, loc=<1.7736 0 0>, rot=<0 0 0 1>}
 tentacle_mid_04_l {parent=tentacle_mid_03_l, loc=<1.5957 0 0>, rot=<0 0 0 1>}
 tentacle_mid_05_l {parent=tentacle_mid_04_l, loc=<1.3864 0 0>, rot=<0 0 0 1>}
 tentacle_upper_01_l {parent=Big_pice, loc=<-4.372 -0.0622 -0.0687>, rot=<-0.2042 -0.6824 -0.6631 0.23>}
 tentacle_upper_02_l {parent=tentacle_upper_01_l, loc=<2.5526 0 0>, rot=<0 0 0 1>}
 tentacle_upper_03_l {parent=tentacle_upper_02_l, loc=<2.5019 0 0>, rot=<0 0 0 1>}
 tentacle_upper_04_l {parent=tentacle_upper_03_l, loc=<3.2258 0 0>, rot=<0 0 0 1>}
 tentacle_upper_05_l {parent=tentacle_upper_04_l, loc=<1.5359 0 0>, rot=<0 0 0 1>}
 tentacle_upper_06_l {parent=tentacle_upper_05_l, loc=<1.2122 0 0>, rot=<0 0 0 1>}
 tentacle_upper_01_r {parent=Big_pice, loc=<4.717 0.5232 -0.1484>, rot=<0.6654 0.2128 0.2303 -0.6774>}
 tentacle_upper_02_r {parent=tentacle_upper_01_r, loc=<2.5526 0 0>, rot=<0 0 0 1>}
 tentacle_upper_03_r {parent=tentacle_upper_02_r, loc=<2.5019 0 0>, rot=<0 0 0 1>}
 tentacle_upper_04_r {parent=tentacle_upper_03_r, loc=<3.0125 0 0>, rot=<0 0 0 1>}
 tentacle_upper_05_r {parent=tentacle_upper_04_r, loc=<1.9243 0 0>, rot=<0 0 0 1>}
 tentacle_upper_06_r {parent=tentacle_upper_05_r, loc=<1.758 0 0>, rot=<0 0 0 1>}
 tentacle_midup_01_r {parent=Big_pice, loc=<6.1223 0.5614 -2.81>, rot=<0.6949 0.0716 0.0861 -0.7103>}
 tentacle_midup_02_r {parent=tentacle_midup_01_r, loc=<1.9387 0 0>, rot=<0 0 0 1>}
 tentacle_midup_03_r {parent=tentacle_midup_02_r, loc=<1.8881 0 0>, rot=<0 0 0 1>}
 tentacle_midup_04_r {parent=tentacle_midup_03_r, loc=<2.3986 0 0>, rot=<0 0 0 1>}
 tentacle_midup_05_r {parent=tentacle_midup_04_r, loc=<1.5772 0 0>, rot=<0 0 0 1>}
 tentacle_midup_06_r {parent=tentacle_midup_05_r, loc=<1.2754 0 0>, rot=<0 0 0 1>}
 tentacle_upper_01_l_smol {parent=Big_pice, loc=<-4.3876 -0.3662 -0.1042>, rot=<0.161 0.6661 0.6749 -0.2736>}
 tentacle_upper_02_l_smol {parent=tentacle_upper_01_l_smol, loc=<1.4126 0 0>, rot=<0 0 0 1>}
 tentacle_mid_01_l_smol {parent=Big_pice, loc=<-5.9397 0.6238 -4.1151>, rot=<-0.3429 0.4254 0.8207 0.1671>}
 tentacle_mid_02_l_smol {parent=tentacle_mid_01_l_smol, loc=<2.3227 0.0525 0>, rot=<0 0 -0.5385 -0.8426>}
 tentacle_mid_03_l_smol {parent=tentacle_mid_02_l_smol, loc=<1.1438 0 0>, rot=<0 0 -0.4324 -0.9017>}
 tentacle_bot_01_l_smol {parent=Big_pice, loc=<-4.4311 0.4173 -8.1383>, rot=<0.3212 -0.6655 -0.6378 -0.2172>}
 tentacle_bot_02_l_smol {parent=tentacle_bot_01_l_smol, loc=<1.1056 -0.033 0>, rot=<0 0 0.1856 -0.9826>}
 tentacle_bot_03_l_smol {parent=tentacle_bot_02_l_smol, loc=<1.1507 0 0>, rot=<0 0 0 1>}
 tentacle_bot_01_r_smol {parent=Big_pice, loc=<4.242 0.154 -9.4928>, rot=<0.5574 -0.2233 -0.5172 -0.6099>}
 tentacle_bot_02_r_smol {parent=tentacle_bot_01_r_smol, loc=<1.2296 0 0>, rot=<0 -0.0332 0 -0.9994>}
 tentacle_bot_03_r_smol {parent=tentacle_bot_02_r_smol, loc=<1.3002 0 0>, rot=<-0.2405 -0.1465 0.4991 -0.8195>}
 tentacle_middv_01_r_smol {parent=Big_pice, loc=<5.5486 0.0816 -6.269>, rot=<0.6337 -0.3199 -0.1873 -0.679>}
 tentacle_middv_02_r_smol {parent=tentacle_middv_01_r_smol, loc=<2.0497 0 0>, rot=<0 0 0 1>}
 tentacle_midup_01_r_smol {parent=Big_pice, loc=<6.4994 0.1321 -2.7327>, rot=<0.6871 0.1262 0.1419 -0.7013>}
 tentacle_midup_02_r_smol {parent=tentacle_midup_01_r_smol, loc=<1.5101 0 0>, rot=<0 0 -0.2916 -0.9565>}
 tentacle_upper_01_r_smol {parent=Big_pice, loc=<4.7797 0.5241 -0.233>, rot=<0.6396 0.1009 0.2213 -0.7292>}
 tentacle_upper_02_r_smol {parent=tentacle_upper_01_r_smol, loc=<2.4511 0 0>, rot=<0 0.133 0 -0.9911>}
 Ppicaxe_01_l {parent=pack1, loc=<-11 0 0>, rot=<0 0 0 1>}
 Ppicaxe_02_l {parent=Ppicaxe_01_l, loc=<3.2534 0 10.3992>, rot=<0 -0.2322 0 -0.9727>}
 Ppicaxe_tentkl_06_l_base {parent=Ppicaxe_02_l, loc=<1.6761 0.2709 3.7289>, rot=<0.999 0.0009 0.0233 0.038>}
 Ppicaxe_tentkl_06_l {parent=Ppicaxe_tentkl_06_l_base, loc=<0 0 0>, rot=<0 0 0 1>}
 Ppicaxe_tentkl_05_l {parent=Ppicaxe_tentkl_06_l, loc=<2.013 0.1562 0.0119>, rot=<0 0 0 1>}
 Ppicaxe_tentkl_04_l_base {parent=Ppicaxe_02_l, loc=<2.787 0.1222 1.2969>, rot=<0.9912 -0.0048 -0.1269 0.0377>}
 Ppicaxe_tentkl_04_l {parent=Ppicaxe_tentkl_04_l_base, loc=<0 0 0>, rot=<0 0 0 1>}
 Ppicaxe_tentkl_03_l {parent=Ppicaxe_tentkl_04_l, loc=<2.3113 -0.0374 -0.0028>, rot=<0 -0.051 0 -0.9987>}
 Ppicaxe_tentkl_02_l_base {parent=Ppicaxe_02_l, loc=<2.8592 -0.039 -2.3351>, rot=<0.9902 -0.0051 -0.1341 0.0376>}
 Ppicaxe_tentkl_02_l {parent=Ppicaxe_tentkl_02_l_base, loc=<0 0 0>, rot=<0 0 0 1>}
 Ppicaxe_tentkl_01_l {parent=Ppicaxe_tentkl_02_l, loc=<2.391 -0.1586 0.1937>, rot=<0 0.1733 0 -0.9849>}
 Ppicaxe_01_r {parent=pack1, loc=<11 0 0>, rot=<0 0 0 1>}
 Ppicaxe_02_r {parent=Ppicaxe_01_r, loc=<-2.5266 0.0475 10.8051>, rot=<-0.2322 -0.0009 0.9727 -0.0036>}
 Ppicaxe_tentkl_01_r_base {parent=Ppicaxe_02_r, loc=<2.0199 0 3.7449>, rot=<0.999 0.0009 0.0233 0.038>}
 Ppicaxe_tentkl_01_r {parent=Ppicaxe_tentkl_01_r_base, loc=<0 0 0>, rot=<0 0 0 1>}
 Ppicaxe_tentkl_02_r {parent=Ppicaxe_tentkl_01_r, loc=<2.2053 0 0>, rot=<0 0 0 1>}
 Ppicaxe_tentkl_03_r_base {parent=Ppicaxe_02_r, loc=<3.299 -0.1306 1.1076>, rot=<0.9958 -0.0032 -0.083 0.0378>}
 Ppicaxe_tentkl_03_r {parent=Ppicaxe_tentkl_03_r_base, loc=<0 0 0>, rot=<0 0 0 1>}
 Ppicaxe_tentkl_04_r {parent=Ppicaxe_tentkl_03_r, loc=<2.6745 0 -0.0371>, rot=<0 -0.051 0 -0.9987>}
 Ppicaxe_tentkl_05_r_base {parent=Ppicaxe_02_r, loc=<3.4957 -0.2291 -2.4267>, rot=<0.9923 -0.0045 -0.1175 0.0377>}
 Ppicaxe_tentkl_05_r {parent=Ppicaxe_tentkl_05_r_base, loc=<0 0 0>, rot=<0 0 0 1>}
 Ppicaxe_tentkl_06_r {parent=Ppicaxe_tentkl_05_r, loc=<2.4656 -0.0856 0.2249>, rot=<0 0.1733 0 -0.9849>}
]
variables [
 isSkyGive {type=bool, default=False}
 pelvis {type=FTransform, default=(Rotation=(X=0,Y=0,Z=0,W=1),Translation=(X=0,Y=0,Z=0),Scale3D=(X=1,Y=1,Z=1))}
 spine_03 {type=FTransform, default=(Rotation=(X=0,Y=0,Z=0,W=1),Translation=(X=0,Y=0,Z=0),Scale3D=(X=1,Y=1,Z=1))}
 spine_04 {type=FTransform, default=(Rotation=(X=0,Y=0,Z=0,W=1),Translation=(X=0,Y=0,Z=0),Scale3D=(X=1,Y=1,Z=1))}
 spine_05 {type=FTransform, default=(Rotation=(X=0,Y=0,Z=0,W=1),Translation=(X=0,Y=0,Z=0),Scale3D=(X=1,Y=1,Z=1))}
]
