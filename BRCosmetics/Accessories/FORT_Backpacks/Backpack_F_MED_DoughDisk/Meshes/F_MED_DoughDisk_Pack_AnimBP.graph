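# Animation Blueprint for the DoughDisk backpack cosmetic, animating its pickaxes and tentacles on the character's back.
# Spring-based AnimDynamics adds secondary motion to pickaxe and pickaxe-tentacle bones, then feeds a Control Rig driven by the body's pelvis and spine transforms plus a skydive flag.
# Fixed additive bone offsets afterwards bend the other tentacles.
# Weapon changes and customization load broadcast FXBAckpack, reporting whether a DoughDisk-tagged pickaxe is equipped.
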AnimBlueprint /BRCosmetics/Accessories/FORT_Backpacks/Backpack_F_MED_DoughDisk/Meshes/F_MED_DoughDisk_Pack_AnimBP
target_skeleton=&/F_MED_DoughDisk_Pack_Skeleton
graphs [
 AnimGraph {nodes [
   n1 = "Transform (Modify) Bone - Bone: Ppicaxe_01_l"(ComponentPose=Local To Component(LocalPose=Input Pose()), Alpha=1, Translation="0,0,3", Rotation="0,0,0", Scale="1,1,1") {
    rotationMode=BMM_Additive
    translationMode=BMM_Additive
    alphaBoolBlend {blendOption=Linear}
    boneToModify {boneName=Ppicaxe_01_l}
    translation {z=3}
   }
   n2 = "Transform (Modify) Bone - Bone: Ppicaxe_01_r"(ComponentPose=@n1, Alpha=1, Translation="0,0,3", Rotation="0,0,0", Scale="1,1,1") {
    rotationMode=BMM_Additive
    translationMode=BMM_Additive
    alphaBoolBlend {blendOption=Linear}
    boneToModify {boneName=Ppicaxe_01_r}
    translation {z=3}
   }
   n3 = Anim Dynamics - Bone: Ppicaxe_tentkl_06_l!(ComponentPose=@n2, Alpha=1, GravityOverride:Vector="0,0,0", ExternalForce="0,0,0") -> (Pose) {
    angularDampingOverride=1
    angularSpringConstant=1440
    bAngularSpring=true
    bGravityOverrideInSimSpace=true
    bOverrideAngularDamping=true
    bUseGravityOverride=true
    simulationSpace=RootRelative
    alphaBoolBlend {blendOption=Linear}
    boundBone {boneName=Ppicaxe_tentkl_06_l}
    componentAppliedLinearAccClamp=<10000 10000 10000>
    componentLinearAccScale=<1 1 1>
    componentLinearVelScale=<0.8 0.8 0.8>
    physicsBodyDefinitions [{
      boundBone {boneName=Ppicaxe_tentkl_06_l}
      boxExtents=<15 15 15>
      constraintSetup {angularLimitsMax=<0 20 5>, angularLimitsMin=<0 -10 -5>, angularTarget=<1 0 0>}
      localJointOffset=<0.5 0 0>
     }]
   }
   n4 = Anim Dynamics - Bone: Ppicaxe_tentkl_04_l_base!(ComponentPose=@n3, Alpha=1, GravityOverride:Vector="0,0,0", ExternalForce="0,0,0") -> (Pose) {
    angularDampingOverride=1
    angularSpringConstant=1440
    bAngularSpring=true
    bGravityOverrideInSimSpace=true
    bOverrideAngularDamping=true
    bUseGravityOverride=true
    simulationSpace=RootRelative
    alphaBoolBlend {blendOption=Linear}
    boundBone {boneName=Ppicaxe_tentkl_04_l_base}
    componentAppliedLinearAccClamp=<10000 10000 10000>
    componentLinearAccScale=<1 1 1>
    componentLinearVelScale=<0.8 0.8 0.8>
    physicsBodyDefinitions [{
      boundBone {boneName=Ppicaxe_tentkl_04_l_base}
      boxExtents=<15 15 15>
      constraintSetup {angularLimitsMax=<0 20 5>, angularLimitsMin=<0 -10 -5>, angularTarget=<1 0 0>}
      localJointOffset=<0.55 0 0>
     }]
   }
   n5 = Anim Dynamics - Bone: Ppicaxe_tentkl_02_l_base!(ComponentPose=@n4, Alpha=1, GravityOverride:Vector="0,0,0", ExternalForce="0,0,0") -> (Pose) {
    angularDampingOverride=1
    angularSpringConstant=1600
    bAngularSpring=true
    bGravityOverrideInSimSpace=true
    bOverrideAngularDamping=true
    bUseGravityOverride=true
    simulationSpace=RootRelative
    alphaBoolBlend {blendOption=Linear}
    boundBone {boneName=Ppicaxe_tentkl_02_l_base}
    componentAppliedLinearAccClamp=<10000 10000 10000>
    componentLinearAccScale=<1 1 1>
    componentLinearVelScale=<0.8 0.8 0.8>
    physicsBodyDefinitions [{
      boundBone {boneName=Ppicaxe_tentkl_02_l_base}
      boxExtents=<15 15 15>
      constraintSetup {angularLimitsMax=<0 1 5>, angularLimitsMin=<0 -15 -5>, angularTarget=<1 0 0>}
      localJointOffset=<0.4 0 0>
     }]
   }
   n6 = Anim Dynamics - Bone: Ppicaxe_tentkl_01_r_base!(ComponentPose=@n5, Alpha=1, GravityOverride:Vector="0,0,0", ExternalForce="0,0,0") -> (Pose) {
    angularDampingOverride=1
    angularSpringConstant=1440
    bAngularSpring=true
    bGravityOverrideInSimSpace=true
    bOverrideAngularDamping=true
    bUseGravityOverride=true
    simulationSpace=RootRelative
    alphaBoolBlend {blendOption=Linear}
    boundBone {boneName=Ppicaxe_tentkl_01_r_base}
    componentAppliedLinearAccClamp=<10000 10000 10000>
    componentLinearAccScale=<1 1 1>
    componentLinearVelScale=<0.8 0.8 0.8>
    physicsBodyDefinitions [{
      boundBone {boneName=Ppicaxe_tentkl_01_r_base}
      boxExtents=<15 15 15>
      constraintSetup {angularLimitsMax=<0 20 5>, angularLimitsMin=<0 -10 -5>, angularTarget=<1 0 0>}
      localJointOffset=<0.5 0 0>
     }]
   }
   n7 = Anim Dynamics - Bone: Ppicaxe_tentkl_03_r_base!(ComponentPose=@n6, Alpha=1, GravityOverride:Vector="0,0,0", ExternalForce="0,0,0") -> (Pose) {
    angularDampingOverride=1
    angularSpringConstant=1440
    bAngularSpring=true
    bGravityOverrideInSimSpace=true
    bOverrideAngularDamping=true
    bUseGravityOverride=true
    simulationSpace=RootRelative
    alphaBoolBlend {blendOption=Linear}
    boundBone {boneName=Ppicaxe_tentkl_03_r_base}
    componentAppliedLinearAccClamp=<10000 10000 10000>
    componentLinearAccScale=<1 1 1>
    componentLinearVelScale=<0.8 0.8 0.8>
    physicsBodyDefinitions [{
      boundBone {boneName=Ppicaxe_tentkl_03_r_base}
      boxExtents=<15 15 15>
      constraintSetup {angularLimitsMax=<0 20 5>, angularLimitsMin=<0 -10 -5>, angularTarget=<1 0 0>}
      localJointOffset=<0.5 0 0>
     }]
   }
   n8 = Anim Dynamics - Bone: Ppicaxe_tentkl_05_r_base!(ComponentPose=@n7, Alpha=1, GravityOverride:Vector="0,0,0", ExternalForce="0,0,0") -> (Pose) {
    angularDampingOverride=1
    angularSpringConstant=1600
    bAngularSpring=true
    bGravityOverrideInSimSpace=true
    bOverrideAngularDamping=true
    bUseGravityOverride=true
    simulationSpace=RootRelative
    alphaBoolBlend {blendOption=Linear}
    boundBone {boneName=Ppicaxe_tentkl_05_r_base}
    componentAppliedLinearAccClamp=<10000 10000 10000>
    componentLinearAccScale=<1 1 1>
    componentLinearVelScale=<0.8 0.8 0.8>
    physicsBodyDefinitions [{
      boundBone {boneName=Ppicaxe_tentkl_05_r_base}
      boxExtents=<15 15 15>
      constraintSetup {angularLimitsMax=<0 15 5>, angularLimitsMin=<0 -1 -5>, angularTarget=<1 0 0>}
      localJointOffset=<0.5 0 0>
     }]
   }
   n9 = Anim Dynamics - Bone: Ppicaxe_01_l(ComponentPose=@n8, Alpha=1, ExternalForce="0,0,0") {
    angularSpringConstant=1
    bLinearSpring=true
    bOverrideLinearDamping=true
    bUseGravityOverride=true
    linearDampingOverride=1
    linearSpringConstant=300
    simulationSpace=RootRelative
    alphaBoolBlend {blendOption=Linear}
    boundBone {boneName=Ppicaxe_01_l}
    componentLinearVelScale=<0.8 0.8 0.8>
    gravityOverride {z=-980}
    physicsBodyDefinitions [{
      boundBone {boneName=Ppicaxe_01_l}
      boxExtents=<1 1 7>
      constraintSetup {angularLimitsMax=<0 0 0>, angularLimitsMin=<0 0 0>, angularTarget=<0 0 0>, linearAxesMax=<2 1 8>, linearAxesMin=<-2 -1 -8>}
      localJointOffset=<0 0 0>
     }]
   }
   n10 = Anim Dynamics - Bone: Ppicaxe_01_r(ComponentPose=@n9, Alpha=1, ExternalForce="0,0,0") {
    angularSpringConstant=1
    bLinearSpring=true
    bOverrideLinearDamping=true
    bUseGravityOverride=true
    linearDampingOverride=1
    linearSpringConstant=400
    simulationSpace=RootRelative
    alphaBoolBlend {blendOption=Linear}
    boundBone {boneName=Ppicaxe_01_r}
    componentLinearVelScale=<0.8 0.8 0.8>
    gravityOverride {z=-980}
    physicsBodyDefinitions [{
      boundBone {boneName=Ppicaxe_01_r}
      boxExtents=<10 10 10>
      constraintSetup {angularLimitsMax=<0 0 0>, angularLimitsMin=<0 0 0>, angularTarget=<0 0 0>, linearAxesMax=<6 5 8>, linearAxesMin=<-6 -5 -8>}
      localJointOffset=<0 0 0>
     }]
   }
   n11 = Control Rig!(
    Source=Component To Local(ComponentPose=@n10)
    Alpha
    pelvis:Transform=Get Pelvis() -> (Pelvis:Transform)
    spine_05:Transform=Get Spine5() -> (Spine5:Transform)
    isSkyGive:bool=Get bIsSkydiveFloating()
    spine_03:Transform=Get spine_03() -> (spine_03:Transform)
    spine_04:Transform=Get spine_04() -> (spine_04:Transform)
   ) -> (Pose) {
    alphaBoolBlend {blendOption=Linear}
    controlRigAssetReference {blueprintRigClass=&/F_MED_DoughDisk_Pack_CtrlRig.F_MED_DoughDisk_Pack_CtrlRig_C}
   }
   n12 = "Transform (Modify) Bone - Bone: pack"(ComponentPose=Local To Component(LocalPose=@n11), Translation="0,0,0", Rotation="0,0,0") {rotationMode=BMM_Additive, translationMode=BMM_Additive, alphaBoolBlend {blendOption=Linear}, boneToModify {boneName=pack}}
   n13 = "Transform (Modify) Bone - Bone: tentacle_upper_04_r"(ComponentPose=@n12, Alpha=1, Translation="0,-1,0", Rotation="-8.197,57.805,8.282", Scale="1,1,1") {
    rotationMode=BMM_Additive
    translationMode=BMM_Additive
    alphaBoolBlend {blendOption=Linear}
    boneToModify {boneName=tentacle_upper_04_r}
    rotation=<rot -8.197 57.805 8.282>
    translation {y=-1}
   }
   n14 = "Transform (Modify) Bone - Bone: tentacle_upper_05_r"(ComponentPose=@n13, Alpha=1, Translation="0,1,0", Rotation="-14.255,67.497,5.824", Scale="1,1,1") {
    rotationMode=BMM_Additive
    translationMode=BMM_Additive
    alphaBoolBlend {blendOption=Linear}
    boneToModify {boneName=tentacle_upper_05_r}
    rotation=<rot -14.255 67.497 5.824>
    translation {y=1}
   }
   n15 = "Transform (Modify) Bone - Bone: tentacle_upper_06_r"(ComponentPose=@n14, Alpha=1, Translation="0,0,0", Rotation="0,50,0", Scale="1,1,1") {
    rotationMode=BMM_Additive
    translationMode=BMM_Additive
    alphaBoolBlend {blendOption=Linear}
    boneToModify {boneName=tentacle_upper_06_r}
    rotation {yaw=50}
   }
   n16 = "Transform (Modify) Bone - Bone: tentacle_midup_04_r"(ComponentPose=@n15, Alpha=1, Translation="0,-1,0", Rotation="0,60,0", Scale="1,1,1") {
    rotationMode=BMM_Additive
    translationMode=BMM_Additive
    alphaBoolBlend {blendOption=Linear}
    boneToModify {boneName=tentacle_midup_04_r}
    rotation {yaw=60}
    translation {y=-1}
   }
   n17 = "Transform (Modify) Bone - Bone: tentacle_midup_05_r"(ComponentPose=@n16, Alpha=1, Translation="0,1,0", Rotation="0,80,0", Scale="1,1,1") {
    rotationMode=BMM_Additive
    translationMode=BMM_Additive
    alphaBoolBlend {blendOption=Linear}
    boneToModify {boneName=tentacle_midup_05_r}
    rotation {yaw=80}
    translation {y=1}
   }
   n18 = "Transform (Modify) Bone - Bone: tentacle_midup_06_r"(ComponentPose=@n17, Alpha=1, Translation="0,0,0", Rotation="0,50,0", Scale="1,1,1") {
    rotationMode=BMM_Additive
    translationMode=BMM_Additive
    alphaBoolBlend {blendOption=Linear}
    boneToModify {boneName=tentacle_midup_06_r}
    rotation {yaw=50}
   }
   n19 = "Transform (Modify) Bone - Bone: tentacle_middv_03_r"(ComponentPose=@n18, Alpha=1, Translation="1,-0.5,0", Rotation="0,60,0", Scale="1,1,1") {
    rotationMode=BMM_Additive
    translationMode=BMM_Additive
    alphaBoolBlend {blendOption=Linear}
    boneToModify {boneName=tentacle_middv_03_r}
    rotation {yaw=60}
    translation=<1 -0.5>
   }
   n20 = "Transform (Modify) Bone - Bone: tentacle_middv_04_r"(ComponentPose=@n19, Alpha=1, Translation="0,0,0", Rotation="0,56,0", Scale="1,1,1") {
    rotationMode=BMM_Additive
    translationMode=BMM_Additive
    alphaBoolBlend {blendOption=Linear}
    boneToModify {boneName=tentacle_middv_04_r}
    rotation {yaw=56}
   }
   n21 = "Transform (Modify) Bone - Bone: tentacle_middv_05_r"(ComponentPose=@n20, Alpha=1, Translation="0,-1,0", Rotation="0,30,0", Scale="1,1,1") {
    rotationMode=BMM_Additive
    translationMode=BMM_Additive
    alphaBoolBlend {blendOption=Linear}
    boneToModify {boneName=tentacle_middv_05_r}
    rotation {yaw=30}
    translation {y=-1}
   }
   n22 = "Transform (Modify) Bone - Bone: tentacle_bot_04_r"(ComponentPose=@n21, Alpha=1, Translation="0.5,0,0", Rotation="0,60,0", Scale="1,1,1") {
    rotationMode=BMM_Additive
    translationMode=BMM_Additive
    alphaBoolBlend {blendOption=Linear}
    boneToModify {boneName=tentacle_bot_04_r}
    rotation {yaw=60}
    translation {x=0.5}
   }
   n23 = "Transform (Modify) Bone - Bone: tentacle_bot_05_r"(ComponentPose=@n22, Alpha=1, Translation="0,0,0", Rotation="0,56,0", Scale="1,1,1") {
    rotationMode=BMM_Additive
    translationMode=BMM_Additive
    alphaBoolBlend {blendOption=Linear}
    boneToModify {boneName=tentacle_bot_05_r}
    rotation {yaw=56}
   }
   n24 = "Transform (Modify) Bone - Bone: tentacle_upper_04_l"(ComponentPose=@n23, Alpha=1, Translation="0,-1,0", Rotation="0,-45,0", Scale="1,1,1") {
    rotationMode=BMM_Additive
    translationMode=BMM_Additive
    alphaBoolBlend {blendOption=Linear}
    boneToModify {boneName=tentacle_upper_04_l}
    rotation {yaw=-45}
    translation {y=-1}
   }
   n25 = "Transform (Modify) Bone - Bone: tentacle_upper_05_l"(ComponentPose=@n24, Alpha=1, Translation="0,1,0", Rotation="0,-50,0", Scale="1,1,1") {
    rotationMode=BMM_Additive
    translationMode=BMM_Additive
    alphaBoolBlend {blendOption=Linear}
    boneToModify {boneName=tentacle_upper_05_l}
    rotation {yaw=-50}
    translation {y=1}
   }
   n26 = "Transform (Modify) Bone - Bone: tentacle_upper_06_l"(ComponentPose=@n25, Alpha=1, Translation="0,0,0", Rotation="0,-50,0", Scale="1,1,1") {
    rotationMode=BMM_Additive
    translationMode=BMM_Additive
    alphaBoolBlend {blendOption=Linear}
    boneToModify {boneName=tentacle_upper_06_l}
    rotation {yaw=-50}
   }
   n27 = "Transform (Modify) Bone - Bone: tentacle_mid_03_l"(ComponentPose=@n26, Alpha=1, Translation="0,0,0", Rotation="0,0,0", Scale="1,1,1") {rotationMode=BMM_Additive, translationMode=BMM_Additive, alphaBoolBlend {blendOption=Linear}, boneToModify {boneName=tentacle_mid_03_l}}
   n28 = "Transform (Modify) Bone - Bone: tentacle_mid_04_l"(ComponentPose=@n27, Alpha=1, Translation="0,0,0", Rotation="0,-60,0", Scale="1,1,1") {
    rotationMode=BMM_Additive
    translationMode=BMM_Additive
    alphaBoolBlend {blendOption=Linear}
    boneToModify {boneName=tentacle_mid_04_l}
    rotation {yaw=-60}
   }
   n29 = "Transform (Modify) Bone - Bone: tentacle_mid_05_l"(ComponentPose=@n28, Alpha=1, Translation="0,0,0", Rotation="0,-50,0", Scale="1,1,1") {
    rotationMode=BMM_Additive
    translationMode=BMM_Additive
    alphaBoolBlend {blendOption=Linear}
    boneToModify {boneName=tentacle_mid_05_l}
    rotation {yaw=-50}
   }
   n30 = "Transform (Modify) Bone - Bone: tentacle_bot_03_l"(ComponentPose=@n29, Alpha=1, Translation="0,0,0", Rotation="0,-10,0", Scale="1,1,1") {
    rotationMode=BMM_Additive
    translationMode=BMM_Additive
    alphaBoolBlend {blendOption=Linear}
    boneToModify {boneName=tentacle_bot_03_l}
    rotation {yaw=-10}
   }
   n31 = "Transform (Modify) Bone - Bone: tentacle_bot_04_l"(ComponentPose=@n30, Alpha=1, Translation="0,0,0", Rotation="0,-56,0", Scale="1,1,1") {
    rotationMode=BMM_Additive
    translationMode=BMM_Additive
    alphaBoolBlend {blendOption=Linear}
    boneToModify {boneName=tentacle_bot_04_l}
    rotation {yaw=-56}
   }
   n32 = "Transform (Modify) Bone - Bone: tentacle_bot_05_l"(ComponentPose=@n31, Alpha=1, Translation="0,0,0", Rotation="0,-30,0", Scale="1,1,1") {
    rotationMode=BMM_Additive
    translationMode=BMM_Additive
    alphaBoolBlend {blendOption=Linear}
    boneToModify {boneName=tentacle_bot_05_l}
    rotation {yaw=-30}
   }
   Output Pose(Result=Component To Local(ComponentPose=@n32)) {layerGroup=None}
  ]}
 EventGraph {nodes [
   Event BlueprintUpdateAnimation()
   > n1 = Gravity Override State Function DEPRECATED(CharacterPartAnimInstance=Self-Reference(), GameTimeInSeconds=GetGameTimeInSeconds(), GravityOverrideParams=Get Gravity Override Params())
   > Set Gravity Override(Gravity Override=@n1)
   n2 = Get Custom Base Mesh() -> (Custom Base Mesh:Skeletal Mesh Component Object Reference)
   > Set Pelvis(Pelvis:Transform=GetSocketTransform(self=@n2, InSocketName=pelvis, TransformSpace=RTS_World)) -> (Output_Get:Transform)
   > Set Spine5(Spine5:Transform=GetSocketTransform(self=@n2, InSocketName=spine_05, TransformSpace=RTS_World)) -> (Output_Get:Transform)
   > Set spine_04(spine_04:Transform=GetSocketTransform(self=@n2, InSocketName=spine_04, TransformSpace=RTS_World)) -> (Output_Get:Transform)
   > Set spine_03(spine_03:Transform=GetSocketTransform(self=@n2, InSocketName=spine_03, TransformSpace=RTS_World)) -> (Output_Get:Transform)
   n3 = Event OnWeaponChanged()
   Get Dynamic Material Instance(execute=[@n3.then, @n5.then], SkeletalMeshComponent=GetPartSkeletalMeshComponent())
   > Call FXBAckpack!(
    execute:Exec
    self:F MED Dough Disk Pack Anim BP Object Reference
    EquipedPickaxe:bool=PickaxeItemDefinitionHasAnyMetaTag(Tags=Make Array("[0]"=(TagName="Cosmetics.MetaTags.DoughDisk")), bRequiresEquipped=true)
   ) -> (then:Exec) {is=title K2Node_CallDelegate}
   n4 = Event OnCharacterCustomizationLoaded()
   > n5 = Set Custom Base Mesh(Custom Base Mesh:Skeletal Mesh Component Object Reference=@n4.CustomBaseMesh) -> (Output_Get:Skeletal Mesh Component Object Reference)
   TryGetPawnOwner()
  ]}
 FXBAckpack {nodes [FXBAckpack!() -> (then, EquipedPickaxe:bool=~) {is=title K2Node_FunctionEntry}]}
]
variables [
 Custom Base Mesh {type=SkeletalMeshComponent}
 FXBAckpack {default=()}
 Gravity Override {type=Vector}
 Gravity Override Params {type=GravityOverrideParamsStruct}
 Pelvis {type=Transform, default {location=<0 0 0>, rotation=<rot 0 0 0>, scale=<1 1 1>}}
 SkyDiv {type=boolean}
 Spine5 {type=Transform, default {location=<0 0 0>, rotation=<rot 0 0 0>, scale=<1 1 1>}}
 spine_03 {type=Transform, default {location=<0 0 0>, rotation=<rot 0 0 0>, scale=<1 1 1>}}
 spine_04 {type=Transform, default {location=<0 0 0>, rotation=<rot 0 0 0>, scale=<1 1 1>}}
]
